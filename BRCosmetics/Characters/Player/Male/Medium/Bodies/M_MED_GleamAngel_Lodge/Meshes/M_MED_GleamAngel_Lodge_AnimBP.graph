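# Animation Blueprint for the GleamAngel Lodge male medium body cosmetic, copying the attached parent mesh's pose and adding secondary motion.
# The pose feeds a Control Rig that receives a Cape Equipped flag, then spring-driven, root-relative Anim Dynamics on the radio and knife joints.
# On customization load, a long or cape backbling sets that flag and zeroes Knife Dyn Alpha, though the knife node shows a constant alpha.
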AnimBlueprint /BRCosmetics/Characters/Player/Male/Medium/Bodies/M_MED_GleamAngel_Lodge/Meshes/M_MED_GleamAngel_Lodge_AnimBP
target_skeleton=&/M_MED_GleamAngel_Lodge_Skeleton
graphs [
 AnimGraph {nodes [
   n1 = Control Rig!(Source=Copy Pose From Mesh() {bUseAttachedParent=true}, Alpha, Cape Equipped:bool=Get Cape Equipped() -> (Cape Equipped:bool)) -> (Pose) {
    alphaBoolBlend {blendOption=Linear}
    controlRigAssetReference {blueprintRigClass=&/M_MED_GleamAngel_Lodge_CtrlRig.M_MED_GleamAngel_Lodge_CtrlRig_C}
   }
   n2 = Anim Dynamics - Bone: dyn_radio_jnt(ComponentPose=Local To Component(LocalPose=@n1), Alpha=1, ExternalForce="0,0,0") {
    angularDampingOverride=1
    angularSpringConstant=20000
    bAngularSpring=true
    bOverrideAngularDamping=true
    bUseGravityOverride=true
    gravityScale=0.35
    simulationSpace=RootRelative
    alphaBoolBlend {blendOption=Linear}
    boundBone {boneName=dyn_radio_jnt}
    componentLinearAccScale=<1 1 1>
    componentLinearVelScale=<1 1 1>
    gravityOverride {z=-980}
    physicsBodyDefinitions [{
      boundBone {boneName=dyn_radio_jnt}
      boxExtents=<40 30 50>
      constraintSetup {angularLimitsMax=<7 1 5>, angularLimitsMin=<-7 -1 -5>, angularTarget=<1 0 0>}
      localJointOffset=<0 0 0>
     }]
   }
   n3 = Anim Dynamics - Bone: dyn_knife_jnt(ComponentPose=@n2, Alpha=1, ExternalForce="0,0,0") {
    angularDampingOverride=1
    angularSpringConstant=50000
    bAngularSpring=true
    bOverrideAngularDamping=true
    bUseGravityOverride=true
    gravityScale=0.35
    simulationSpace=RootRelative
    alphaBoolBlend {blendOption=Linear}
    boundBone {boneName=dyn_knife_jnt}
    componentLinearAccScale=<1 1 1>
    componentLinearVelScale=<1 1 1>
    gravityOverride {z=-980}
    physicsBodyDefinitions [{
      boundBone {boneName=dyn_knife_jnt}
      boxExtents=<80 80 80>
      constraintSetup {angularLimitsMax=<5 2.5 2.5>, angularLimitsMin=<-20 -2.5 -2.5>, angularTarget=<1 0 0>}
      localJointOffset=<0.05 0 0>
     }]
   }
   Output Pose(Result=Component To Local(ComponentPose=@n3)) {layerGroup=None}
  ]}
 EventGraph {nodes [
   Event OnCharacterCustomizationLoaded()
   n1 = CheckForBackbling()
   > n2 = Set Cape Equipped(Cape Equipped:bool=OR Boolean(A=@n1.bIsBackblingACape, B=@n1.bIsBackblingLong)) -> (Output_Get:bool)
   > n3 = Branch(Condition=@n2)
   > Set Knife Dyn Alpha(Knife Dyn Alpha:double=0) -> (Output_Get:double)
   Set Knife Dyn Alpha(execute=@n3.else, Knife Dyn Alpha=1)
  ]}
]
variables [
 Cape Equipped {type=boolean}
 Knife Dyn Alpha {type=number}
]
